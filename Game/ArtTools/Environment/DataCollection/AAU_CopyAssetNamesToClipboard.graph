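# Editor utility Blueprint that copies the names of the assets selected in the editor to the clipboard.
# CopyNamesToClipboard calls GetAssetNames, which adds each selected asset's name uniquely to AR_AssetNames through CreateAssetData.
# BuildListText then appends those names to STR_AssetNames, one per line, and the trailing-trimmed result goes to the clipboard.
# No graph clears either variable before filling it.
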EditorUtilityBlueprint /Game/ArtTools/Environment/DataCollection/AAU_CopyAssetNamesToClipboard
graphs [
 BuildListText {nodes [
   BuildListText() {is=title K2Node_FunctionEntry}
   > n1 = For Each Loop(Array:Array of Names=Get AR_AssetNames() -> (AR_AssetNames:Array of Names)) -> (Array Element:Name)
   n2 = "To String (Name)"(InName=@n1)
   > Set STR_AssetNames(STR_AssetNames:String=Append(A=Get STR_AssetNames() -> (STR_AssetNames:String), B=Append(A=@n2, B="\r\n"))) -> (Output_Get:String)
  ]}
 CopyNamesToClipboard {nodes [
   CopyNamesToClipboard() {is=title K2Node_FunctionEntry}
   > GetAssetNames()
   > BuildListText() {is=title K2Node_CallFunction}
   > CopyStringToClipboard!(execute, InString:String=TrimTrailing!(SourceString:String=Get STR_AssetNames()) -> (ReturnValue:String)) -> (then)
  ]}
 EventGraph {links [], nodes []}
 GetAssetNames {nodes [
   GetAssetNames() {is=title K2Node_FunctionEntry}
   > n1 = GetSelectedAssets()
   > n2 = For Each Loop(Array:Array of Object References=@n1) -> (Array Element:Object Reference)
   > Add Unique(
    TargetArray:Array of Names=Get AR_AssetNames()
    NewItem:"Name (by ref)"=Break Asset Data(AssetData=CreateAssetData(InAsset=@n2, bAllowBlueprintClass=false)).AssetName
   )
  ]}
]
variables [
 AR_AssetNames {type="array<string>", default []}
 STR_AssetNames {type=string, default=""}
]
